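# Scoring row widget for the tournament details screen, pairing a place or threshold on the left with points or a prize value on the right.
# PreConstruct uses UsePlaceIcon to construct a material icon or styled text in the left grid cell, each removing the other.
# Setters fill both sides from text, integer or dollar-formatted float.
# SetPlace validity-checks CommonText_Points but writes to CommonText_Place, which is absent in icon mode.
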WidgetBlueprint /Compete/TournamentDetails/Sections/WBP_TournamentDetailsSectionItem_Scoring
graphs [
 AddPlaceIcon {nodes [
   AddPlaceIcon() {is=title K2Node_FunctionEntry}
   > RemovePlaceText()
   > n1 = Construct Object from Class(Class=&Image) -> (ReturnValue:Image Object Reference)
   > n2 = Set Image_PlaceIcon(Image_PlaceIcon:Image Object Reference=@n1) -> (Output_Get:Image Object Reference)
   > n3 = AddChildToGrid!(
    execute
    self:Grid Panel Object Reference=Get Root() -> (Root:Grid Panel Object Reference)
    Content:Widget Object Reference=@n2
    InRow:int=0
    InColumn:int=0
   ) -> (then, ReturnValue:Grid Slot Object Reference)
   > SetPlaceSlotSettings!(execute, self, Grid Slot:Grid Slot Object Reference=@n3) -> (then)
   > SetDesiredSizeOverride!(
    execute
    self:Image Object Reference=Get Image_PlaceIcon() -> (Image_PlaceIcon:Image Object Reference)
    DesiredSize:Vector 2D Structure=Get IconSize() -> (IconSize:Vector 2D Structure)
   ) -> (then)
   > SetBrushFromMaterial(self=Get Image_PlaceIcon(), Material=Get IconMaterial() -> (IconMaterial:Material Interface Object Reference))
  ]}
 AddPlaceText {nodes [
   AddPlaceText() {is=title K2Node_FunctionEntry}
   > RemovePlaceImage()
   > n1 = Construct Object from Class!(
    execute
    Class=&CommonTextBlock
    self
    bIsScrollingEnabled:bool=True
    Style:Common Text Style Class Reference
    ScrollStyle:Common Text Scroll Style Class Reference=&TextScrollStyle-Base_C
    ScrollOrientation:EOrientation Enum=Orient_Horizontal
   ) -> (then, ReturnValue:Common Text Object Reference)
   > n2 = Set CommonText_Place(CommonText_Place:Common Text Object Reference=@n1) -> (Output_Get:Common Text Object Reference)
   > n3 = AddChildToGrid(self=Get Root(), Content=@n2, InRow=0, InColumn=0)
   > SetPlaceSlotSettings(Grid Slot=@n3)
   > UpdateWidgetStyleTags!(
    execute
    InWidget:Widget Object Reference=Get CommonText_Place() -> (CommonText_Place:Common Text Object Reference)
    InAddedTags:Array of Gameplay Tag Structures=Get PlaceTextStyleTags() -> (PlaceTextStyleTags:Array of Gameplay Tag Structures)
    InRemovedTags:Array of Gameplay Tag Structures=Make Array()
   ) -> (then, ReturnValue:bool)
   > SetColorAndOpacity(self=Get CommonText_Place(), InColorAndOpacity=Get PlaceTextColor() -> (PlaceTextColor:Slate Color Structure))
  ]}
 EventGraph {nodes [
   Event PreConstruct()
   > n1 = Branch(Condition=Get UsePlaceIcon() -> (UsePlaceIcon:bool))
   > AddPlaceIcon() {is=title K2Node_CallFunction}
   AddPlaceText(execute=@n1.else) {is=title K2Node_CallFunction}
  ]}
 RemovePlaceImage {nodes [
   RemovePlaceImage() {is=title K2Node_FunctionEntry}
   n1 = Get Image_PlaceIcon()
   > Is Valid(InputObject=@n1)
   > RemoveFromParent(self=@n1)
   > Set Image_PlaceIcon()
   > Set UsePlaceIcon(UsePlaceIcon:bool=false) -> (Output_Get:bool)
  ]}
 RemovePlaceText {nodes [
   RemovePlaceText() {is=title K2Node_FunctionEntry}
   n1 = Get CommonText_Place()
   > Is Valid(InputObject=@n1)
   > RemoveFromParent(self=@n1)
   > Set CommonText_Place()
   > Set UsePlaceIcon(UsePlaceIcon=true)
  ]}
 Set Prize Data {nodes [
   n1 = Set Prize Data!() -> (then, In Threshold Text:Text, In Value Text:Text) {is=title K2Node_FunctionEntry}
   > SetPlace!(execute, self, InText:Text=@n1) -> (then)
   > SetPointsText!(execute, self, PointsText:Text=@n1.In Value Text) -> (then)
  ]}
 SetPlace {nodes [
   n1 = SetPlace!() -> (then, InText:Text) {is=title K2Node_FunctionEntry}
   > Is Valid(InputObject=Get CommonText_Points() -> (CommonText_Points:Common Text Object Reference))
   > n2 = Set PlaceText(PlaceText:Text=@n1) -> (Output_Get:Text)
   > "SetText (Text)"(self=Get CommonText_Place(), InText=@n2)
  ]}
 SetPlaceSlotSettings {nodes [
   n1 = SetPlaceSlotSettings!() -> (then, Grid Slot:Grid Slot Object Reference) {is=title K2Node_FunctionEntry}
   > SetHorizontalAlignment(
    self:Grid Slot Object Reference=@n1
    InHorizontalAlignment=Get Place Horizontal Alignment() -> (Place Horizontal Alignment:EHorizontalAlignment Enum=HAlign_Fill)
   )
   > SetVerticalAlignment(
    self:Grid Slot Object Reference=@n1
    InVerticalAlignment=Get Place Vertical Alignment() -> (Place Vertical Alignment:EVerticalAlignment Enum=VAlign_Fill)
   )
   > SetPadding!(execute, self:Grid Slot Object Reference=@n1, InPadding:Margin Structure=Get Place Padding() -> (Place Padding:Margin Structure)) -> (then)
  ]}
 SetPoints {nodes [
   n1 = SetPoints!() -> (then, Points:int) {is=title K2Node_FunctionEntry}
   > n2 = Set PointsText(PointsText:Text="To Text (Integer)"(Value=@n1)) -> (Output_Get:Text)
   > "SetText (Text)"(self=Get CommonText_Points(), InText=@n2)
  ]}
 SetPointsFloat {nodes [
   n1 = SetPointsFloat!() -> (then, Value:double=~) {is=title K2Node_FunctionEntry}
   > "SetText (Text)"(
    self=Get CommonText_Points()
    InText=Format Text!(
     Format="INVTEXT(\"${Cash}\")"
     Cash:Text="To Text (Float)"(Value=@n1, MinimumIntegralDigits=1, MaximumIntegralDigits=324, MinimumFractionalDigits=2, MaximumFractionalDigits=3)
    ) -> (Result)
   )
  ]}
 SetPointsText {nodes [
   n1 = SetPointsText!() -> (then, PointsText:Text) {is=title K2Node_FunctionEntry}
   > n2 = Set PointsText(PointsText=@n1)
   > "SetText (Text)"(self=Get CommonText_Points(), InText=@n2)
  ]}
]
variables [
 CommonText_Place {type=CommonTextBlock, default=null}
 IconMaterial {type=MaterialInterface, default=/FortUILibrary/Textures/MaterialInstances/MI_UIKit_Icon_Compete}
 IconSize {type=Vector2D, default=<32 32>}
 Image_PlaceIcon {type=Image, default=null}
 Place Horizontal Alignment {type=EHorizontalAlignment, default=HAlign_Left}
 Place Padding {type=Margin, default=<ltrb 48 0 -28 0>}
 Place Vertical Alignment {type=EVerticalAlignment, default=VAlign_Center}
 PlaceText {type=string, default="##2"}
 PlaceTextColor {type=SlateColor, default {colorUseRule=UseColor_Specified, specifiedColor=<rgba 0.8 0.8 0.8 1>}}
 PlaceTextStyleTags {type="array<GameplayTag>", default [{tagName=StyleSheet.Style.Label}, {tagName=StyleSheet.Style.10}, {tagName=StyleSheet.Style.Sentence}]}
 PointsText {type=string, default=""}
 UsePlaceIcon {type=boolean}
]
widgets {
 info {inheritedWidgetCount=0, parentClass=&UserWidget, rootWidgetClass=&GridPanel, widgetCount=2}
 widgets [
  Root {widgetClassPath=&GridPanel, bIsVariable=true, widget=Root, properties {columnFill [0,1], rowFill [1]}}
  CommonText_Points {
   widgetClassPath=&CommonTextBlock
   bIsVariable=true
   parent=Root
   widget=CommonText_Points
   properties {
    scrollStyle=&TextScrollStyle-Base_C
    text=0
    font {
     bForceMonospaced=false
     bMaterialIsStencil=false
     fontMaterial=null
     fontObject=&NotoSans
     letterSpacing=0
     monospacedWidth=1
     size=24
     skewAmount=0
     typefaceFontName=NotoSans-Regular
     outlineSettings {
      bApplyOutlineToDropShadows=false
      bMiteredCorners=false
      bSeparateFillAlpha=false
      outlineMaterial=null
      outlineSize=0
      outlineColor=<rgba 0 0 0 1>
     }
    }
    shadowOffset=<0 0>
   }
   slot {class=GridSlot, properties {column=1, horizontalAlignment=HAlign_Right, verticalAlignment=VAlign_Center, padding=<ltrb 0 0 48 0>}}
  }
 ]
}
