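# Base widget for sections of the Compete tournament details screen, a focusable tile button derived from WBP_CompeteTileButton_Base.
# A grid root holds a named SectionSlot for content and a rounded outline block, which PreConstruct registers as a modular button block.
# SetFocusAnim stores a focus animation, alongside timing variables and OnFocused/OnUnfocused members that have no logic here.
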WidgetBlueprint /Compete/TournamentDetails/Sections/WBP_TournamentDetailsSection_Base
graphs [
 EventGraph {nodes [
   Event PreConstruct()
   > RegisterModularBlock!(
    execute
    self:UIKit Modular Button Object Reference
    Block:UIKit Block Object Reference=Get WBP_UIKit_Block_Outline() -> (WBP_UIKit_Block_Outline:WBP UIKit Block Outline Object Reference)
   ) -> (then)
  ]}
 OnFocused {nodes [OnFocused() {is=title K2Node_FunctionEntry}]}
 OnUnfocused {nodes [OnUnfocused() {is=title K2Node_FunctionEntry}]}
 SetFocusAnim {nodes [
   n1 = SetFocusAnim!() -> (then, Anim_Focus:Widget Animation Object Reference) {is=title K2Node_FunctionEntry}
   > Set Anim_Focus(Anim_Focus:Widget Animation Object Reference=@n1) -> (Output_Get:Widget Animation Object Reference)
  ]}
]
variables [
 Anim_Focus {type=WidgetAnimation, default=null}
 FocusingTransitionTime {type=number, default=0.3}
 FocusTime {type=number, default=0.3}
 OnFocused {}
 OnUnfocused {}
 UnfocusTime {type=number, default=0.15}
]
widgets {
 info {
  inheritedWidgetCount=1
  parentClass=/Compete/WBP_CompeteTileButton_Base.WBP_CompeteTileButton_Base_C
  rootWidgetClass=&GridPanel
  widgetCount=4
 }
 widgets [
  Root {widgetClassPath=&GridPanel, bIsVariable=true, widget=Root, properties {columnFill [1], rowFill [1]}}
  SectionSlot {widgetClassPath=&NamedSlot, bIsVariable=true, parent=Root, widget=SectionSlot, slot {class=GridSlot}}
  WBP_UIKit_Block_Outline {
   widgetClassPath=/FortUILibrary/UIKit/Buttons/Blocks/WBP_UIKit_Block_Outline
   bIsVariable=true
   parent=Root
   widget=WBP_UIKit_Block_Outline
   properties {
    bottomLeftCornerRadius=33
    bottomRightCornerRadius=33
    radiusInPixels=true
    topLeftCornerRadius=33
    topRightCornerRadius=33
    padding=<ltrb -10 -10 -10 -10>
   }
   slot {class=GridSlot}
  }
  InputActionWidget {widgetClassPath=&CommonActionWidget, bInherited=true, widget=null}
 ]
}
